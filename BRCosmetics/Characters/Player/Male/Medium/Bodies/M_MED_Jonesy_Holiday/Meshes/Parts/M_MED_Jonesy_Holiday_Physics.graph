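# Physics asset for the Jonesy Holiday male character part, defining secondary motion for the hat's ear flaps and fur.
# Each side chains box bodies from a small flap base through parent, 01 and 02, with limits favoring one swing axis and allowing little twist.
# Four fur capsules attach directly to the head with tight, symmetric swing limits.
# Head and neck capsules supply body shapes; the flap bases have no constraint to the head.
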PhysicsAsset /BRCosmetics/Characters/Player/Male/Medium/Bodies/M_MED_Jonesy_Holiday/Meshes/Parts/M_MED_Jonesy_Holiday_Physics
bodies {
 dyn_hat_flap_01_l [{shapeType=Box, extentX=2.76, extentY=2.76, extentZ=1.01, center=<1.0133 0 0.0001>}]
 dyn_hat_flap_01_r [{shapeType=Box, extentX=2.51, extentY=2.51, extentZ=1.01, center=<1.0103 0 -0.0001>}]
 dyn_hat_flap_02_l [{shapeType=Box, extentX=2.01, extentY=2.76, extentZ=1.01, center=<0.8773 0 0>}]
 dyn_hat_flap_02_r [{shapeType=Box, extentX=2.26, extentY=2.51, extentZ=1.01, center=<0.6512 0 0>}]
 dyn_hat_flap_base_l [{shapeType=Box, extentX=0.5, extentY=0.5, extentZ=0.5}]
 dyn_hat_flap_base_r [{shapeType=Box, extentX=0.5, extentY=0.5, extentZ=0.5}]
 dyn_hat_flap_parent_l [{shapeType=Box, extentX=7.5944, extentY=9.4279, extentZ=2.899, center=<4.0861 0 -0.0002>}]
 dyn_hat_flap_parent_r [{shapeType=Box, extentX=6.9077, extentY=8.6878, extentZ=3.231, center=<3.8075 0 -0.0002>}]
 dyn_hat_fur_01 [{shapeType=Capsule, length=1.255, radius=1.675, center=<0 0 1.2278>, rotation=<rot 90 0 0>}]
 dyn_hat_fur_02 [{shapeType=Capsule, length=2.3898, radius=1.55, center=<-0.6664 0 0.8257>, rotation=<rot 90 0 0>}]
 dyn_hat_fur_03 [{shapeType=Capsule, length=1.005, radius=1.675, center=<0 -0.0001 1.1465>, rotation=<rot 90 0 0>}]
 dyn_hat_fur_04 [{shapeType=Capsule, length=2.38, radius=1.425, center=<0 0 1.0596>, rotation=<rot 90 0 0>}]
 head [{shapeType=Capsule, length=3.375, radius=6.875, center=<0 1.7835 0>, rotation=<rot 90 0 0>}]
 neck_01 [{shapeType=Capsule, length=9.25, radius=6, center=<-2.8347 0 0>}]
}
constraints [
 {bone1Name=dyn_hat_flap_parent_r, bone2Name=dyn_hat_flap_base_r, swing1LimitDegrees=7, swing2LimitDegrees=30, twistLimitDegrees=1}
 {bone1Name=dyn_hat_flap_01_r, bone2Name=dyn_hat_flap_parent_r, swing1LimitDegrees=5, swing2LimitDegrees=25, twistLimitDegrees=3}
 {bone1Name=dyn_hat_flap_02_r, bone2Name=dyn_hat_flap_01_r, swing1LimitDegrees=5, swing2LimitDegrees=25, twistLimitDegrees=3}
 {bone1Name=dyn_hat_flap_parent_l, bone2Name=dyn_hat_flap_base_l, swing1LimitDegrees=7, swing2LimitDegrees=30, twistLimitDegrees=1}
 {bone1Name=dyn_hat_flap_01_l, bone2Name=dyn_hat_flap_parent_l, swing1LimitDegrees=5, swing2LimitDegrees=25, twistLimitDegrees=3}
 {bone1Name=dyn_hat_flap_02_l, bone2Name=dyn_hat_flap_01_l, swing1LimitDegrees=5, swing2LimitDegrees=25, twistLimitDegrees=3}
 {bone1Name=dyn_hat_fur_01, bone2Name=head, swing1LimitDegrees=12, swing2LimitDegrees=12, twistLimitDegrees=3}
 {bone1Name=dyn_hat_fur_02, bone2Name=head, swing1LimitDegrees=12, swing2LimitDegrees=12, twistLimitDegrees=3}
 {bone1Name=dyn_hat_fur_03, bone2Name=head, swing1LimitDegrees=12, swing2LimitDegrees=12, twistLimitDegrees=3}
 {bone1Name=dyn_hat_fur_04, bone2Name=head, swing1LimitDegrees=12, swing2LimitDegrees=12, twistLimitDegrees=3}
]
